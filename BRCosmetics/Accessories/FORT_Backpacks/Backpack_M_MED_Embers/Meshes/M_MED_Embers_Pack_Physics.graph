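# Physics asset for the M_MED_Embers backpack cosmetic, setting up two dangling chains for physics-driven motion.
# Small capsule bodies cover the root, the pack, a six-bone long chain and a three-bone short chain.
# Both chains hang from the pack, each link constrained to the previous with very tight twist and swing limits that generally narrow toward the tips.
# The pack-to-root joint leaves all swing and twist free.
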PhysicsAsset /BRCosmetics/Accessories/FORT_Backpacks/Backpack_M_MED_Embers/Meshes/M_MED_Embers_Pack_Physics
bodies {
 dyn_long_chain_0 [{shapeType=Capsule, length=0.75, radius=1, center=<1.3911 0 0>, rotation=<rot 90 0 0>}]
 dyn_long_chain_1 [{shapeType=Capsule, length=0.75, radius=1, center=<1.3911 0 0>, rotation=<rot 90 0 0>}]
 dyn_long_chain_2 [{shapeType=Capsule, length=0.75, radius=1, center=<1.3911 0 0>, rotation=<rot 90 0 0>}]
 dyn_long_chain_3 [{shapeType=Capsule, length=0.75, radius=1, center=<1.3911 0 0>, rotation=<rot 90 0 0>}]
 dyn_long_chain_4 [{shapeType=Capsule, length=0.75, radius=1, center=<1.3911 0 0>, rotation=<rot 90 0 0>}]
 dyn_long_chain_5 [{shapeType=Capsule, length=0.75, radius=1, center=<1.3911 0 0>, rotation=<rot 90 0 0>}]
 dyn_short_chain_1 [{shapeType=Capsule, length=0.75, radius=1, center=<1.3911 0 0>, rotation=<rot 90 0 0>}]
 dyn_short_chain_2 [{shapeType=Capsule, length=0.75, radius=1, center=<1.3911 0 0>, rotation=<rot 90 0 0>}]
 dyn_short_chain_3 [{shapeType=Capsule, length=0.75, radius=1, center=<1.3911 0 0>, rotation=<rot 90 0 0>}]
 pack [{shapeType=Capsule, length=1, radius=1}]
 root [{shapeType=Capsule, length=1, radius=1}]
}
constraints [
 {
  bone1Name=pack
  bone2Name=root
  swing1LimitDegrees=45
  swing1Motion=Free
  swing2LimitDegrees=45
  swing2Motion=Free
  twistLimitDegrees=45
  twistMotion=Free
 }
 {bone1Name=dyn_short_chain_1, bone2Name=pack, swing1LimitDegrees=45, swing2LimitDegrees=30, twistLimitDegrees=3}
 {bone1Name=dyn_short_chain_2, bone2Name=dyn_short_chain_1, swing1LimitDegrees=45, swing2LimitDegrees=30, twistLimitDegrees=3}
 {bone1Name=dyn_short_chain_3, bone2Name=dyn_short_chain_2, swing1LimitDegrees=45, swing2LimitDegrees=20, twistLimitDegrees=3}
 {bone1Name=dyn_long_chain_0, bone2Name=pack, swing1LimitDegrees=45, swing2LimitDegrees=45, twistLimitDegrees=3}
 {bone1Name=dyn_long_chain_1, bone2Name=dyn_long_chain_0, swing1LimitDegrees=45, swing2LimitDegrees=30, twistLimitDegrees=3}
 {bone1Name=dyn_long_chain_2, bone2Name=dyn_long_chain_1, swing1LimitDegrees=40, swing2LimitDegrees=20, twistLimitDegrees=3}
 {bone1Name=dyn_long_chain_3, bone2Name=dyn_long_chain_2, swing1LimitDegrees=40, swing2LimitDegrees=15, twistLimitDegrees=3}
 {bone1Name=dyn_long_chain_4, bone2Name=dyn_long_chain_3, swing1LimitDegrees=10, swing2LimitDegrees=20, twistLimitDegrees=3}
 {bone1Name=dyn_long_chain_5, bone2Name=dyn_long_chain_4, swing1LimitDegrees=10, swing2LimitDegrees=15, twistLimitDegrees=3}
]
